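; Field system actor that creates a vacuum effect, applying linear velocity to physics objects around it in the direction of its owner.
; On BeginPlay, only on the remote (non-authority) side and with a valid owner, it stores the normalized direction to the owner and starts MagnitudeTimeline.
; Each timeline update adds a transient field multiplying a linear radial falloff by that fixed direction vector, scaled by the timeline's magnitude.
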(event EventBeginPlay
  (bind _returnvalue (Actor|GetOwner))
  (|CallParentFunction)
  (Networking|SwitchHasAuthority
    (:Remote
      (Utilities|IsValid _returnvalue
        (:Is%20Valid
          (Variables|Default|SetOwnerDirection (Math|Vector|Normalize (- (Transformation|GetActorLocation _returnvalue) (Transformation|GetActorLocation))))
          (bind (_direction _magnitude) (Timeline|MagnitudeTimeline PlayFromStart)
            (:Update
              (Field|AddTransientField (Variables|Field|GetFieldSystemComponent) true "Field_LinearVelocity" 0 (Field|SetOperatorField (Variables|Default|GetOperatorField) _magnitude (Field|SetRadialFalloff (Variables|Default|GetRadialFalloff) 1.0 0.0 1.0 0.0 1.2e+03 (Transformation|GetWorldLocation (Variables|Field|GetFieldSystemComponent)) "Field_Falloff_Linear") (Field|SetUniformVector (Variables|Default|GetUniformVector) 1.0 (Variables|Default|GetOwnerDirection)) "Field_Multiply")))))))))

(event Collision|EventActorBeginOverlap
  (params (OtherActor :type Actor%20Object%20Reference)))

(fn ConstructionScript)
