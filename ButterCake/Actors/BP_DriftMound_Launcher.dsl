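; Calendar-event geyser launcher that snaps to a referenced Buttercake drift mound, destroying itself if the event is inactive or mound missing.
; A random authority timer or player trigger overlap starts prime, fire and reset timers, with RepNotify states driving collision, visuals and audio.
; Cone overlaps, gated by ShouldBoost, launch non-skydiving players, boost vehicles and invert or offset projectile, C4 and toy gravity.
; Exiting players and vehicles ignore their next fall damage.
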(event EventBeginPlay
  ;[0.05098 1 0.01223] Begin Play
  (Class|FortAsyncActionAuthorityMatchReady|MatchReadyAsync
    (:OnReady
      (MCPContext|InitialCalendarSyncCompleted
        (:OnSyncComplete
          (bind _biseventactive (CalendarEvents|GetCalendarEventInformation (Variables|Default|GetEvent%20Name)))
          (if _biseventactive
            (Variables|Default|SetisDisabled true)
            (CallFunction|SetupMaterialParams)
            (CallFunction|SetupGeyserWait)
            (Variables|Default|SetSavedConeScale (Variables|/Script/Engine.SceneComponent|GetRelativeScale3D (Variables|Default|GetS_FX_GeyserCone)))
            ;[1 0 0] No Buttercake Mound Referenced--Destroy Actor
            (Variables|Default|GetDriftMound_BC
              (:then)
              (:else
                (Actor|DestroyActor)
                (Development|PrintString "Must reference Artemis_DriftMound_BC" false true "(R=1.000000,G=0.000000,B=0.000000,A=1.000000)")))
            (else
              ;[1 0 0] Calendar Event Inactive--Destroy Actor
              (Actor|DestroyActor)
              (Development|PrintString "CALENDAR EVENT NOT ACTIVE" false true "(R=1.000000,G=0.000000,B=0.000000,A=1.000000)"))))))))

(event Custom|SetupMaterialParams
  (bind _s_fx_geysercone (Variables|Default|GetS_FX_GeyserCone))
  ;[0.008445 0.4764 1] Set material values
  (Rendering|Material|SetScalarParameterValueonMaterials _s_fx_geysercone "UVScale" (Math|Random|RandomFloatinRange 2.0 4.0))
  (Rendering|Material|SetScalarParameterValueonMaterials _s_fx_geysercone "SlantAmount" (Math|Random|RandomFloatinRange -5e+02 5e+02))
  (Rendering|Material|SetScalarParameterValueonMaterials _s_fx_geysercone "WPO_Amount" (Math|Random|RandomFloatinRange 0.0 5e+02))
  (Rendering|Material|SetScalarParameterValueonMaterials _s_fx_geysercone "WorldTexUVScale" (Math|Random|RandomFloatinRange 0.5 2.0))
  (bind _output_get (Variables|Default|SetGeyserContrast (Math|Random|RandomFloatinRange 1.0 3.0)))
  (Rendering|Material|SetScalarParameterValueonMaterials _s_fx_geysercone "Top Edge Contrast" _output_get))

(event Custom|SetupGeyserWait
  ;[1 0 0.9359] Timed Firing. Setup the "wait" time before the geyser fires. Prime the geyser to allow players to enter/exit. Fire to launch the players. Reset the flow.
  (Networking|SwitchHasAuthority
    (:Authority
      (Utilities|Time|SetTimerbyEvent (delegate Custom|PrimeGeyser) (Math|Random|RandomFloatinRange (ScalableFloat|GetValueatLevel (Variables|Default|GetSF_MinActivationTime)) (ScalableFloat|GetValueatLevel (Variables|Default|GetSF_MaxActivationTime)))))))

(event OnComponentBeginOverlap(S_FX_GeyserCone)
  (params (OverlappedComponent :type Primitive%20Component%20Object%20Reference) (OtherActor :type Actor%20Object%20Reference) (OtherComp :type Primitive%20Component%20Object%20Reference) (OtherBodyIndex :type Integer) (bFromSweep :type Boolean) (SweepResult :type Hit%20Result%20Structure%20%28by%20ref%29))
  (bind (_returnvalue_x _returnvalue_y _returnvalue_z) (Math|Vector|BreakVector (Transformation|GetActorLocation)))
  ;[1 0.3378 0] Launch players and vehicles if they overlap the geyser while it's active. Handle projectiles and toys/C4 separately.
  (if (Variables|Default|GetShouldBoost)
    (bind _asfort_player_pawn (Utilities|Casting|CastToFortPlayerPawn OtherActor)
      (:then
        (if (not (GameplayTags|HasTag (GameplayTags|GetOwnedGameplayTags _asfort_player_pawn :!class /Script/GameplayTags.GameplayTagAssetInterface) "(TagName=\"MovementMode.Skydiving\")"))
          (Networking|SwitchHasAuthority
            (:Authority
              (Ability|SendGameplayEventtoActor _asfort_player_pawn "(TagName=\"Abilities.Generic.Vent.Enter\")" (Utilities|Struct|MakeGameplayEventData :Target self :ContextHandle "()" :InstigatorTags "(GameplayTags=)" :TargetTags "(GameplayTags=)" :EventMagnitude (Variables|Default|GetUpwardLaunchVelocity) :TargetData "()"))
              (Character|LaunchCharacter _asfort_player_pawn (* (Transformation|GetActorUpVector) (ScalableFloat|GetValueatLevel (Variables|Default|GetSF_LaunchVelocity))) (Variables|Default|GetXYOverride) (Variables|Default|GetZOverride)))
            (:Remote
              (Utilities|GeyserPlayerLaunchSound _asfort_player_pawn)
              (bind (_returnvalue_x_2 _returnvalue_y_2 _returnvalue_z_2) (Math|Vector|BreakVector (Transformation|GetActorLocation _asfort_player_pawn)))
              (Projectile|AddGameplayCue%28Looping%29Local _asfort_player_pawn (Variables|Default|GetVentCue) (Ability|GameplayCue|MakeGameplayCueParameters :NormalizedMagnitude (Math|Conversions|ToFloat(Boolean) (<= _returnvalue_z_2 (+ _returnvalue_z 8343.651367))) :RawMagnitude 0.0 :EffectContext "()" :AggregatedSourceTags "(GameplayTags=)" :AggregatedTargetTags "(GameplayTags=)" :Location "0,0,0" :Normal "0,0,0" :GameplayEffectLevel 1 :AbilityLevel 1 :bReplicateLocationWhenUsingMinimalRepProxy false))))))
      (:CastFailed
        (bind _asfort_vehicle_interface (Utilities|Casting|CastToFortVehicleInterface OtherActor)
          (:then
            (CallFunction|OnVehicleEnterZone (Vehicle|GetActor _asfort_vehicle_interface) -0.6)
            (Vehicle|SetIgnoreAllFallingDamage _asfort_vehicle_interface true)
            (Audio|BroadcastAbilityImpactatLocation 0 (Vehicle|GetActor _asfort_vehicle_interface) (+ (Vehicle|GetVehicleVelocity _asfort_vehicle_interface) (* (Transformation|GetActorUpVector) (ScalableFloat|GetValueatLevel (Variables|Default|GetSF_LaunchVelocity)))) 777.0))
          (:CastFailed
            (bind _asfort_projectile_base (Utilities|Casting|CastToFortProjectileBase OtherActor)
              (:then
                (CallFunction|ProjEnter _asfort_projectile_base))
              (:CastFailed
                (bind _asb_bga_athena_c4 (Utilities|Casting|CastToB_BGA_Athena_C4 OtherActor)
                  (:then
                    (Variables|/Script/Engine.ProjectileMovementComponent|SetProjectileGravityScale (* (Variables|/Script/Engine.ProjectileMovementComponent|GetProjectileGravityScale (Variables|/Game/Athena/Items/Consumables/C4/B_BGA_Athena_C4.B_BGA_Athena_C4_C|GetProjectileMovement1 _asb_bga_athena_c4)) -1.0) (Variables|/Game/Athena/Items/Consumables/C4/B_BGA_Athena_C4.B_BGA_Athena_C4_C|GetProjectileMovement1 _asb_bga_athena_c4)))
                  (:CastFailed
                    (bind _asphysics_ball_master (Utilities|Casting|CastToPhysicsBall_Master OtherActor))
                    (CallFunction|ToyEnter _asphysics_ball_master)))))))))))

(event OnComponentBeginOverlap(TriggerVolume)
  (params (OverlappedComponent :type Primitive%20Component%20Object%20Reference) (OtherActor :type Actor%20Object%20Reference) (OtherComp :type Primitive%20Component%20Object%20Reference) (OtherBodyIndex :type Integer) (bFromSweep :type Boolean) (SweepResult :type Hit%20Result%20Structure%20%28by%20ref%29))
  ;[0 1 1] Trigger geyser immediately if player overlaps
  (Utilities|Casting|CastToFortPlayerPawn OtherActor)
  (CallFunction|PrimeGeyser))

(event Custom|PrimeGeyser
  ;[1 0 0.9359] Timed Firing. Setup the "wait" time before the geyser fires. Prime the geyser to allow players to enter/exit. Fire to launch the players. Reset the flow.
  (Variables|Default|SetPrimeMode true)
  (Utilities|Time|SetTimerbyEvent (delegate Custom|FireGeyser) (ScalableFloat|GetValueatLevel (Variables|Default|GetSF_PrimeDuration))))

(event Custom|FireGeyser
  ;[1 0 0.9359] Timed Firing. Setup the "wait" time before the geyser fires. Prime the geyser to allow players to enter/exit. Fire to launch the players. Reset the flow.
  (Variables|Default|SetPrimeMode false)
  (Variables|Default|SetisDisabled false)
  (Utilities|Time|SetTimerbyEvent (delegate Custom|ResetGeyserWait) (ScalableFloat|GetValueatLevel (Variables|Default|GetSF_ActiveDuration))))

(event Custom|ResetGeyserWait
  ;[1 0 0.9359] Timed Firing. Setup the "wait" time before the geyser fires. Prime the geyser to allow players to enter/exit. Fire to launch the players. Reset the flow.
  (Variables|Default|SetisDisabled true)
  (CallFunction|SetupGeyserWait))

(event OnComponentEndOverlap(S_FX_GeyserCone)
  (params (OverlappedComponent :type Primitive%20Component%20Object%20Reference) (OtherActor :type Actor%20Object%20Reference) (OtherComp :type Primitive%20Component%20Object%20Reference) (OtherBodyIndex :type Integer))
  ;[0.1134 0 1] Set the launched targets to be immune to fall damage.
  (if (Variables|Default|GetShouldBoost)
    (bind _asfort_player_pawn (Utilities|Casting|CastToFortPlayerPawn OtherActor)
      (:then
        (Ability|SendGameplayEventtoActor _asfort_player_pawn "(TagName=\"Abilities.Generic.Vent.Exit\")")
        (Variables|/Script/FortniteGame.FortPawn|SetbIgnoreNextFallingDamage true _asfort_player_pawn)
        (Projectile|RemoveGameplayCue%28Looping%29Local _asfort_player_pawn (Variables|Default|GetVentCue) (Ability|GameplayCue|MakeGameplayCueParameters :NormalizedMagnitude 0.0 :RawMagnitude 0.0 :EffectContext "()" :AggregatedSourceTags "(GameplayTags=)" :AggregatedTargetTags "(GameplayTags=)" :Location "0,0,0" :Normal "0,0,0" :GameplayEffectLevel 1 :AbilityLevel 1 :bReplicateLocationWhenUsingMinimalRepProxy false)))
      (:CastFailed
        (bind _asfort_vehicle_interface (Utilities|Casting|CastToFortVehicleInterface OtherActor)
          (:then
            (CallFunction|OnVehicleExitZone (Vehicle|GetActor _asfort_vehicle_interface))
            (Vehicle|SetIgnoreAllFallingDamage _asfort_vehicle_interface false)
            (Vehicle|SetIgnoreNextFallingDamage _asfort_vehicle_interface true))
          (:CastFailed
            (bind _asfort_projectile_base (Utilities|Casting|CastToFortProjectileBase OtherActor)
              (:then
                (CallFunction|ProjExit _asfort_projectile_base))
              (:CastFailed
                (bind _asb_bga_athena_c4 (Utilities|Casting|CastToB_BGA_Athena_C4 OtherActor)
                  (:then
                    (Variables|/Script/Engine.ProjectileMovementComponent|SetProjectileGravityScale (* (Variables|/Script/Engine.ProjectileMovementComponent|GetProjectileGravityScale (Variables|/Game/Athena/Items/Consumables/C4/B_BGA_Athena_C4.B_BGA_Athena_C4_C|GetProjectileMovement1 _asb_bga_athena_c4)) -1.0) (Variables|/Game/Athena/Items/Consumables/C4/B_BGA_Athena_C4.B_BGA_Athena_C4_C|GetProjectileMovement1 _asb_bga_athena_c4)))
                  (:CastFailed
                    (bind _asphysics_ball_master (Utilities|Casting|CastToPhysicsBall_Master OtherActor))
                    (CallFunction|ToyExit _asphysics_ball_master)))))))))))

(fn GetCollisionStaticMesh
  (return "0, 0, 0" "/Game/Athena/Environments/Island/Runes/Mesh/SM_CorruptedVent_A.SM_CorruptedVent_A"))

(fn GetMeshComponents
  (return 0))

(fn IgnoreWorldGridScaleOverride
  (return false))

(fn OnRep_IsActive)

(fn OnRep_isDisabled
  (if (Variables|Default|GetisDisabled)
    (Collision|SetCollisionEnabled (Variables|Default|GetS_FX_GeyserCone) "NoCollision")
    (Components|Activation|Deactivate (Variables|Default|GetP_GeyserSteamBottom))
    (Rendering|SetVisibility (Variables|Default|GetS_FX_GeyserCone) false false)
    (else
      (Collision|SetCollisionEnabled (Variables|Default|GetS_FX_GeyserCone) "QueryAndPhysics")
      (Components|Activation|Activate (Variables|Default|GetLoopingAmbAudio) false)
      (Rendering|SetVisibility (Variables|Default|GetS_FX_GeyserCone) true false)
      (Niagara|SpawnSystematLocation :SystemTemplate (Variables|Default|GetN_SneezeVisual) :Location (Transformation|GetActorLocation) :Scale "1.000000,1.000000,1.000000" :bAutoDestroy false :bAutoActivate true :PoolingMethod "None" :bPreCullCheck true)
      (Utilities|GeyserActivationSound))))

(fn OnRep_PrimeMode
  (bind _p_geysersteambottom (Variables|Default|GetP_GeyserSteamBottom))
  (if (Variables|Default|GetPrimeMode)
    (Components|Activation|Activate _p_geysersteambottom false)
    (Audio|PlaySoundatLocation (Variables|Default|GetSoundOnPrime) (Transformation|GetActorLocation))
    (else
      (Components|Activation|Deactivate _p_geysersteambottom))))

(fn ProjEnter
  (params (Projectile :type Fort%20Projectile%20Base%20Object%20Reference))
  (bind _projectilemovementcomponent (Variables|/Script/FortniteGame.FortProjectileBase|GetProjectileMovementComponent Projectile))
  (bind _projectilegravityscale (Variables|/Script/Engine.ProjectileMovementComponent|GetProjectileGravityScale _projectilemovementcomponent))
  (if (> _projectilegravityscale 0.0)
    (Variables|/Script/Engine.ProjectileMovementComponent|SetProjectileGravityScale (* _projectilegravityscale -1.0) _projectilemovementcomponent)
    (else
      (Variables|/Script/Engine.ProjectileMovementComponent|SetProjectileGravityScale (+ _projectilegravityscale (Variables|Default|GetZeroGravPrj)) _projectilemovementcomponent))))

(fn ProjExit
  (params (Projectile :type Fort%20Projectile%20Base%20Object%20Reference))
  (bind _projectilemovementcomponent (Variables|/Script/FortniteGame.FortProjectileBase|GetProjectileMovementComponent Projectile))
  (bind _projectilegravityscale (Variables|/Script/Engine.ProjectileMovementComponent|GetProjectileGravityScale _projectilemovementcomponent))
  (if (== _projectilegravityscale (Variables|Default|GetZeroGravPrj))
    (Variables|/Script/Engine.ProjectileMovementComponent|SetProjectileGravityScale (- _projectilegravityscale (Variables|Default|GetZeroGravPrj)) _projectilemovementcomponent)
    (else
      (Variables|/Script/Engine.ProjectileMovementComponent|SetProjectileGravityScale (* _projectilegravityscale -1.0) _projectilemovementcomponent))))

(fn ToyEnter
  (params (Projectile :type Physics%20Ball%20Master%20Object%20Reference))
  (bind _projectilemovement (Variables|/Game/Building/ActorBlueprints/Prop/PhysicsBall_Master.PhysicsBall_Master_C|GetProjectileMovement Projectile))
  (bind _projectilegravityscale (Variables|/Script/Engine.ProjectileMovementComponent|GetProjectileGravityScale _projectilemovement))
  (if (> _projectilegravityscale 0.0)
    (Variables|/Script/Engine.ProjectileMovementComponent|SetProjectileGravityScale (* _projectilegravityscale -0.5) _projectilemovement)
    (else
      (Variables|/Script/Engine.ProjectileMovementComponent|SetProjectileGravityScale (+ _projectilegravityscale (Variables|Default|GetZeroGravPrj)) _projectilemovement))))

(fn ToyExit
  (params (Projectile :type Physics%20Ball%20Master%20Object%20Reference))
  (bind _projectilemovement (Variables|/Game/Building/ActorBlueprints/Prop/PhysicsBall_Master.PhysicsBall_Master_C|GetProjectileMovement Projectile))
  (bind _projectilegravityscale (Variables|/Script/Engine.ProjectileMovementComponent|GetProjectileGravityScale _projectilemovement))
  (if (== _projectilegravityscale (Variables|Default|GetZeroGravPrj))
    (Variables|/Script/Engine.ProjectileMovementComponent|SetProjectileGravityScale (- _projectilegravityscale (Variables|Default|GetZeroGravPrj)) _projectilemovement)
    (else
      (Variables|/Script/Engine.ProjectileMovementComponent|SetProjectileGravityScale (* _projectilegravityscale -2.0) _projectilemovement))))

(fn ConstructionScript
  (|CallParentFunction)
  (Utilities|FlowControl|Sequence
    (:then_0
      (Rendering|Material|SetVectorParameterValueonMaterials (Variables|Default|GetS_FX_GeyserCone) "SulfurBaseColor" (Math|Conversions|ToVector(LinearColor) (Variables|LocalVFXColor|GetGeyser%20Illuminated%20Mesh%20Color))))
    (:then_1
      ;[0 1 1] Move geyser to referenced blowhole location
      (bind _driftmound_bc (Variables|Default|GetDriftMound_BC)
        (:then
          (Transformation|SetActorLocation (Transformation|GetWorldLocation (Variables|/Game/Athena/Artemis/Environments/BuildingActors/SnowDrift/Artemis_DriftMound_BC.Artemis_DriftMound_BC_C|GetGeyserAttach (Utilities|ResolveSoftReference _driftmound_bc))) false true))
        (:else)))))
